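# Unlit, two-sided additive material for Niagara mesh and sprite particles in the squares aura effect.
# A sine band along V, narrowed by a coverage parameter, forms a soft glow plus a thresholded hot core.
# An erosion parameter dissolves the core through randomly offset plasma-cloud noise and fades the glow.
# The mask, scaled by a brightness parameter, drives alpha and blends a blue tint, both multiplied by particle color.
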Material /SparksCosmetics/Items/Auras/Assets/NS_Aura_Squares/M_Aura_Squares_Mesh
graph {
 properties {
  bAutomaticallySetUsageInEditor=false
  bCastRayTracedShadows=false
  bUsedWithNiagaraMeshParticles=true
  bUsedWithNiagaraSprites=true
  blendMode=BLEND_Additive
  previewMesh=&/SM_Aura_Squares_Flat
  shadingModel=MSM_Unlit
  twoSided=true
  usedShadingModels=Unlit
 }
 nodes [
  n1 = ParticleColor()
  n2 = TextureCoordinate()
  n3 = ParticleRandom()
  n4 = TextureSample(UVs=Multiply(A=Constant2Vector() {g=0.125, r=2.5}, B=Add(A=@n2, B=@n3))) {texture=/Game/Effects/Fort_Effects/Textures/T_VFX_PlasmaClouds}
  n5 = DynamicParameter() {defaultValue=<rgba 1 0.5 0.5 0.5>, paramNames ["Brightness Mult (pre-saturate)", "Square Coverage (0-1)", Erosion, --]}
  n6 = LinearInterpolate(
   A=LinearInterpolate(B=@n4.G, Alpha=@n5.Erosion) {constA=1}
   B=LinearInterpolate(A=@n4.G, Alpha=@n5.Erosion) {constA=1, constB=0}
   Alpha=@n5.Erosion
  ) {constA=1}
  n7 = ComponentMask(None=TextureCoordinate()) {g=true}
  n8 = DynamicParameter() {defaultValue=<rgba 1 0.5 0.5 0.5>, paramNames ["Brightness Mult (pre-saturate)", "Square Coverage (0-1)", Erosion, --]}
  n9 = MaterialFunctionCall(
   Input=Sine(None=Multiply(A=@n7) {constB=0.5})
   Input Low=OneMinus(None=Multiply(A=@n8."Square Coverage (0-1)", B=@n8."Square Coverage (0-1)"))
   Input High=Constant() {r=1}
  ) {materialFunction=&RemapValue_0-1}
  n10 = Constant2Vector() {g=0.99, r=0.95}
  n11 = MaterialFunctionCall(Input=@n9.Result, Input Low=@n10.R, Input High=@n10.R) {materialFunction=&RemapValue_0-1}
  n12 = Add(
   A=Multiply(A=ScalarParameter() {defaultValue=1, parameterName=Use Hot Core}, B=Multiply(A=@n6, B=@n11.Result))
   B=Multiply(A=Multiply(A=@n9.Result) {constB=0.2}, B=OneMinus(None=@n8.Erosion))
  ) {constB=0}
  n13 = DynamicParameter() {defaultValue=<rgba 1 0.5 0.5 0.5>, paramNames ["Brightness Mult (pre-saturate)", "Square Coverage (0-1)", Erosion, --]}
  Multiply(A=@n1.A, B=Saturate(None=Multiply(A=@n12, B=@n13."Brightness Mult (pre-saturate)")))
  Multiply(A=LinearInterpolate(A=Constant3Vector() {constant=<rgba 0.3579 0.7421 1.4 1>}, Alpha=@n12), B=@n1.RGB)
 ]
}
