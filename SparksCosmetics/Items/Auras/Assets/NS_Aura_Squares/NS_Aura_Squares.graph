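# Niagara character aura cosmetic that pulses colored squares around the body, with User flags adjusting for large bodies and frontend brightness.
# A system-level bone position per loop anchors Squares: growing thick and thin mesh/sprite bursts, sharing loop-seeded offsets and one of five palettes.
# Glows holds bone-tracking sprites with bone-pair midpoints, time-of-day scaling and smoothed loop-seeded colors.
# Decal continuously spawns bone-attached, camera-offset sprites that fade in and out.
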NiagaraSystem /SparksCosmetics/Items/Auras/Assets/NS_Aura_Squares/NS_Aura_Squares
dependencies {
 usedDataInterfaces [
  &NiagaraDataInterfaceArrayFloat
  &NiagaraDataInterfaceArrayInt32
  &NiagaraDataInterfaceCurve
  &NiagaraDataInterfaceArrayColor
  &NiagaraDataInterfaceColorCurve
  &NiagaraDataInterfaceSkeletalMesh
 ]
 usedDynamicInputs [
  &MakeVector
  &MakeFloatFromVector
  &RandomRangeInt
 ]
 usedModules [
  &self:SystemSpawnScript.SystemScriptSource.SystemScriptGraph.NiagaraNodeAssignment_0.SetVariables_#1
  &self:SystemSpawnScript.SystemScriptSource.SystemScriptGraph.NiagaraNodeAssignment_1.SetVariables_#2
  &self:SystemSpawnScript.SystemScriptSource.SystemScriptGraph.NiagaraNodeAssignment_2.SetVariables_#3
  &SystemState
  &SampleSkeletalMesh
  &self:BonePosThisLoop
 ]
}
emitters {
 Decal {
  emitterSpawnScript [/SetVariables(Emitter.VisibleDepth:NiagaraFloat=&MakeCustomFloatFromBool(Bool=$User.IsLargeBody, True Float=40, False Float=20)) {isSetParameters=true}]
  emitterUpdateScript [
   &EmitterState()
   &SpawnRate(SpawnRate=8, Use Spawn Probability=false)
   &SpawnBurst_Instantaneous(Spawn Count=4, Use Spawn Probability=false, UseLoopCountLimit=true)
  ]
  particleSpawnScript [
   &FN_SetSpawnGroup()
   &InitializeParticle(
    Color Mode=::NewEnumerator0
    Sprite Rotation Angle=&RandomRangeFloat(Minimum=-30, Maximum=30)
    Sprite Rotation Mode=::NewEnumerator0
    Sprite Size Mode=::NewEnumerator3
    Uniform Sprite Size=&SelectFloatFromArray(
     Direct Array Index=&SpawnGroup()
     Float Selection Array=DataInterface() {propertyValues {FloatData [35,35,30,30,35,35]}}
    )
    UsePositionOffset=false
    Write Material Random=true
   ) {module=InitializeParticle001}
  ]
  particleUpdateScript [
   &ParticleState()
   &CameraOffset(Camera Offset Amount=&MakeCustomFloatFromBool(Bool=$User.IsLargeBody, True Float=30, False Float=20))
   &SkeletalMeshLocation(
    Bone / Socket Index=&SelectIntFromArray(Direct Array Index=$Particles.SpawnGroup, Int Selection Array=DataInterface() {propertyValues {IntData [1,2,3,4]}})
    Bone Sampling Mode=::NewEnumerator4
    MaskSpawnGroup=false
    Skeletal Mesh=$System.Skeletal Mesh
   ) {module=SkeletalMeshLocation001}
   &DynamicMaterialParameters(
    Index 0 Param 1=2
    Index 0 Param 2=&RandomRangeFloat(Minimum=0, Maximum=1)
    Index 0 Param 3=&RandomRangeFloat(Minimum=0, Maximum=1)
    Index 0 Param 4=0.25
    Module.Param0WriteEnabled=true
    Module.Param1WriteEnabled=true
    Module.Param2WriteEnabled=true
    Module.Param3WriteEnabled=true
   )
   &Color(Color=&MakeLinearColorFromVectorAndFloat(
     "Vector (RGB)"=&Lerp_Vector(
      A=&RandomVector(Vector Scale=&MakeCustomFloatFromBool(Bool=$User.IsFrontend, True Float=5, False Float=20))
      B=&MakeVectorFromLinearColorRGB(Color=&MakeCustomLinearColorFromBool(Bool=$User.IsFrontend, True Color=<rgba 0.169 0.7221 1 1>, False Color=<rgba 1.69 7.2211 10 1>))
      Alpha=&FloatFromCurve(FloatCurve=DataInterface() {propertyValues {Curve {keys [
           {interpMode=RCIM_Cubic, leaveTangentWeight=0.6551, tangentMode=RCTM_User, time=0, value=0}
           {arriveTangentWeight=0.1667, interpMode=RCIM_Cubic, leaveTangentWeight=0.1667, tangentMode=RCTM_User, time=1, value=1}
          ]}}})
     )
     "Float (Alpha)"=&FloatFromCurve(FloatCurve=DataInterface() {propertyValues {Curve {keys [
          {arriveTangent=4.4103, interpMode=RCIM_Cubic, leaveTangent=4.4103, leaveTangentWeight=0.6551, tangentMode=RCTM_User, time=0, value=0}
          {arriveTangentWeight=0.1667, interpMode=RCIM_Cubic, leaveTangentWeight=0.1667, tangentMode=RCTM_User, time=0.5, value=1}
          {arriveTangent=-4.0948, tangentMode=RCTM_User, time=1, value=0}
         ]}}})
    ))
  ]
  rendererClasses [&NiagaraSpriteRendererProperties]
  renderers [
   {rendererClass=&NiagaraSpriteRendererProperties}
   {rendererClass=&NiagaraSpriteRendererProperties, rendererIndex=1}
  ]
 }
 Glows {
  emitterUpdateScript [
   &EmitterState()
   &SpawnBurst_Instantaneous(Spawn Count=10, Use Spawn Probability=false, UseLoopCountLimit=true) {module=SpawnBurst_Instantaneous001}
  ]
  particleSpawnScript [
   /SetVariables(
    Particles.PickBone:NiagaraInt32=&SelectIntFromArray(
     Direct Array Index=&ReturnExecIndex()
     Int Selection Array=DataInterface() {propertyValues {IntData [1,1,2,2,3,3,4,4,5,6]}}
    )
    Particles.PickBoneMatch:NiagaraInt32=&SelectIntFromArray(
     Direct Array Index=&ReturnExecIndex()
     Int Selection Array=DataInterface() {propertyValues {IntData [1,3,2,4,3,5,4,6,5,6]}}
    )
    Particles.ScaleGroup:NiagaraInt32=&SelectIntFromArray(
     Direct Array Index=&ReturnExecIndex()
     Int Selection Array=DataInterface() {propertyValues {IntData [0,1,0,1,2,3,2,3,4,4]}}
    )
   ) {isSetParameters=true}
   &InitializeParticle(
    Lifetime=2
    Position Offset=<0 -30 110>
    Sprite Rotation Mode=::NewEnumerator2
    Sprite Size Mode=::NewEnumerator3
    Uniform Sprite Size=&Multiply_Float(
     A=&SelectFloatFromArray(Direct Array Index=$Particles.ScaleGroup, Float Selection Array=DataInterface() {propertyValues {FloatData [45,42,40,35,30]}})
     B=&MakeCustomFloatFromBool(Bool=$User.IsLargeBody, True Float=1.35, False Float=1)
    )
    UsePositionOffset=false
    Write Material Random=true
   )
  ]
  particleUpdateScript [
   &ParticleState(Kill Particles When Lifetime Has Elapsed=false, Let Infinitely Lived Particles Die When Emitter Deactivates=true)
   &SpriteRotationRate(Rotation Rate=&MaskFloatBySpawnGroup(
     Masked Float=&SelectFloatFromArray(
      Direct Array Index=&ReturnExecIndex()
      Float Selection Array=DataInterface() {propertyValues {FloatData [-15,20]}}
     )
     Passthrough Float=0
    ))
   /SetVariables(Transient.ToDMult:NiagaraFloat=&FNTimeOfDayScalar(
     Night Scale=1.5
     Morning Scale=1.2
     Day Scale=&MakeCustomFloatFromBool(Bool=$User.IsFrontEnd, True Float=1, False Float=1.65)
     Evening Scale=0.9
    )) {isSetParameters=true}
   &InterpolateOverTime(
    Target Color Value:LinearColor=&SelectLinearColorFromArray(
     Array Sampling Mode=::NewEnumerator0
     Direct Array Index=?
     Evaluation Type=::NewEnumerator1
     Color Selection Array=DataInterface() {propertyValues {ColorData [<rgba 1 0 0 1>, <rgba 0 1 0 1>, <rgba 0 0 1 1>]}}
     Randomness Mode=::NewEnumerator1
     Random Seed=&Add_Integer(A=$System.SystemRandom, B=&Add_Integer(A=$System.LoopCount, B=$Engine.ExecIndex))
     Fixed Random Seed=true
    )
    Variable Type=::NewEnumerator4
   )
   /SetVariables(Particles.BodyColor:LinearColor=&MakeLinearColorFromVectorAndFloat(
     "Vector (RGB)"=&MakeVectorFromLinearColorRGB(Color=$Particles.InterpolateOverTime.InterpolateOverTime_LinearColor.Moving Average)
     "Float (Alpha)"=&Multiply_Float(
      A=&Multiply_Float(
       A=$Transient.ToDMult
       B=&MakeCustomFloatFromBool(Bool=$User.IsFrontEnd, True Float=0.6, False Float=1.1)
      )
      B=&Lerp_Float(A=0.5, B=0.25, Alpha=&Multiply_FloatByInt(Float=0.25, Integer=$Particles.ScaleGroup))
     )
    )) {isSetParameters=true}
   &DynamicMaterialParameters(
    Index 0 Param 1=2
    Module.Index0Param0Write=true
    Module.Index0Param1Write=false
    Module.Index0Param2Write=false
    Module.Index0Param3Write=false
   )
   &SkeletalMeshLocation(
    Bone / Socket Index=$Particles.PickBone
    Bone Sampling Mode=::NewEnumerator4
    MaskSpawnGroup=false
    Orientation Sampling=::NewEnumerator1
    Skeletal Mesh=$System.Skeletal Mesh
   )
   &SkeletalMeshLocation(
    Bone / Socket Index=$Particles.PickBoneMatch
    Bone Sampling Mode=::NewEnumerator4
    MaskSpawnGroup=false
    Orientation Sampling=::NewEnumerator1
    Position Sampling=::NewEnumerator2
    Skeletal Mesh=$System.Skeletal Mesh
   ) {module=SkeletalMeshLocation001}
   &CameraOffset(Camera Offset Amount=&SelectFloatFromArray(
     Direct Array Index=&SpawnGroup()
     Float Selection Array=DataInterface() {propertyValues {FloatData [-40,-20,5]}}
    ))
   /SetVariables(Particles.GlowPosition:NiagaraPosition=&LerpPosition(A=$Particles.SkeletalMeshLocation.SampledPosition, B=$Particles.SkeletalMeshLocation001.SampledPosition, Alpha=0.5)) {isSetParameters=true}
   &Color(Color=&Multiply_LinearColorByFloat(
     Linear Color=&ColorFromCurve(
      ColorCurve=DataInterface() {propertyValues {
        AlphaCurve {keys [<at 0 1 RCIM_Cubic>]}
        BlueCurve {keys [<at 0 0 RCIM_Cubic>,<at 0.5 0>,<at 1 1 RCIM_Cubic>]}
        GreenCurve {keys [<at 0 0 RCIM_Cubic>,<at 0.5 1>,<at 1 0 RCIM_Cubic>]}
        RedCurve {keys [<at 0 1 RCIM_Cubic>,<at 0.5 0>,<at 1 0 RCIM_Cubic>]}
       }}
      CurveIndex=&Multiply_FloatByInt(Float=0.25, Integer=$Particles.ScaleGroup)
     )
     Float=1
    )) {enabled=false}
  ]
  rendererClasses [&NiagaraSpriteRendererProperties]
  renderers [{rendererClass=&NiagaraSpriteRendererProperties}]
 }
 Squares {
  emitterUpdateScript [
   &EmitterState()
   &SpawnRate(
    SpawnRate=&FloatFromCurve(
     FloatCurve=DataInterface() {propertyValues {Curve {keys [{arriveTangent=-1.5414, tangentMode=RCTM_User, time=0.25, value=10}, <at 0.26 0>]}}}
     CurveIndex=$Emitter.LoopedAge
     Scale Curve=&RandomRangeFloat(Minimum=0.8, Maximum=1.2)
    )
    Use Spawn Probability=false
   )
   &SpawnBurst_Instantaneous(Spawn Group=1, Use Spawn Probability=false, UseLoopCountLimit=false)
   &SpawnBurst_Instantaneous(
    Spawn Group=1
    Spawn Time=/Niagara/DynamicInputs/Math/Min_Float(A:NiagaraFloat=0.2, B:NiagaraFloat=&Multiply_Float(A=$Emitter.CurrentLoopDuration, B=0.5))
    Use Spawn Probability=false
    UseLoopCountLimit=false
   ) {module=SpawnBurst_Instantaneous001}
  ]
  particleSpawnScript [
   &InitializeParticle(
    Lifetime=&MaskFloatBySpawnGroup(
     Masked Float=&RandomRangeFloat(Minimum=0.75, Maximum=1.5)
     Passthrough Float=&RandomRangeFloat(Minimum=0.75, Maximum=1.25)
    )
    Mesh Scale Mode=::NewEnumerator3
    Mesh Uniform Scale=&RandomRangeFloat(
     Minimum=0.75
     Maximum=2
     Randomness Mode=::NewEnumerator1
     Random Seed=&Add_Integer(A=$System.SystemRandom, B=$System.LoopCount)
     Fixed Random Seed=true
    )
    Position=&Subtract_Position(A=$System.BonePos, B=$Engine.Owner.Position)
    Position Mode=::NewEnumerator1
    Position Offset=&MakeCustomVectorFromBool(
     Bool=$User.IsLargeBody
     True Vector=&RandomRangeVector(
      Minimum=<-45 -30 -30>
      Maximum=<45 30 45>
      Randomness Mode=::NewEnumerator1
      Random Seed=&Add_Integer(A=$System.SystemRandom, B=$System.LoopCount)
      Fixed Random Seed=true
     )
     False Vector=&RandomRangeVector(
      Minimum=<-30 -30 -30>
      Maximum=<30 30 30>
      Randomness Mode=::NewEnumerator1
      Random Seed=&Add_Integer(A=$System.SystemRandom, B=$System.LoopCount)
      Fixed Random Seed=true
     )
    )
    Position Offset Coordinate Space=::Local
    UsePositionOffset=true
    Write Material Random=true
   )
   /SetVariables(Particles.BurstColor:NiagaraInt32=&ModuloInt(A=&Add_Integer(A=$System.SystemRandom, B=$System.LoopCount), B=5)) {isSetParameters=true}
   /SetVariables(Particles.BurstColor=4) {enabled=false, isSetParameters=true}
   /SetVariables(
    Transient.ThickColor:LinearColor=&self:SelectColor(
     "Option 0 (Red)":LinearColor=&SelectLinearColorFromArray(
      Array Sampling Mode=::NewEnumerator0
      Direct Array Index=?
      Color Selection Array=DataInterface() {propertyValues {ColorData [<rgba 1.7 0.1705 0.1705 1>, <rgba 1 0.1478 0.1478 1>]}}
      Random Seed=0
     )
     "Option 1 (Orange)":LinearColor=&SelectLinearColorFromArray(
      Array Sampling Mode=::NewEnumerator0
      Direct Array Index=?
      Color Selection Array=DataInterface() {propertyValues {ColorData [<rgba 1 0.6664 0.0552 0.6>, <rgba 1 0.247 0.0226 0.75>, <rgba 1 0.3555 0.2452 0.6>]}}
      Random Seed=0
     )
     "Option 2 (Yellow)":LinearColor=&SelectLinearColorFromArray(
      Array Sampling Mode=::NewEnumerator0
      Direct Array Index=?
      Color Selection Array=DataInterface() {propertyValues {ColorData [<rgba 1 0.6664 0.0552 0.4>, <rgba 1 1 1 0.25>, <rgba 0.4789 1 0.2452 0.3>]}}
      Random Seed=0
     )
     "Option 3 (Green)":LinearColor=&SelectLinearColorFromArray(
      Array Sampling Mode=::NewEnumerator0
      Direct Array Index=?
      Color Selection Array=DataInterface() {propertyValues {ColorData [<rgba 0.1357 0.3395 1.2 1>, <rgba 1 1 1 0.3>, <rgba 0.273 1 0.6852 0.3>]}}
      Random Seed=0
     )
     "Option 4 (Blue)":LinearColor=&SelectLinearColorFromArray(
      Array Sampling Mode=::NewEnumerator0
      Direct Array Index=?
      Color Selection Array=DataInterface() {propertyValues {ColorData [<rgba 0.3922 0.2452 1 0.75>, <rgba 1 1 1 0.3>, <rgba 0.1687 0.8742 1 0.5>]}}
      Random Seed=0
     )
     Color Select:NiagaraInt32=$Particles.BurstColor
    )
    Transient.ThinColor:LinearColor=&self:SelectColor(
     "Option 0 (Red)":LinearColor=&SelectLinearColorFromArray(
      Array Sampling Mode=::NewEnumerator0
      Direct Array Index=?
      Color Selection Array=DataInterface() {propertyValues {ColorData [<rgba 2 0.2006 0.2306 1>, <rgba 1.5 0.3854 0.347 1>]}}
      Random Seed=0
     )
     "Option 1 (Orange)":LinearColor=&SelectLinearColorFromArray(
      Array Sampling Mode=::NewEnumerator0
      Direct Array Index=?
      Color Selection Array=DataInterface() {propertyValues {ColorData [<rgba 1.5 0.349 0.0828 1>, <rgba 1 0.4464 0.051 0.2>, <rgba 1 0.2875 0.1617 1>]}}
      Random Seed=0
     )
     "Option 2 (Yellow)":LinearColor=&SelectLinearColorFromArray(
      Array Sampling Mode=::NewEnumerator0
      Direct Array Index=?
      Color Selection Array=DataInterface() {propertyValues {ColorData [<rgba 1 0.6664 0.0552 1>, <rgba 1 0.7434 0.273 0.5>, <rgba 0.6195 1 0.273 0.5>]}}
      Random Seed=0
     )
     "Option 3 (Green)":LinearColor=&SelectLinearColorFromArray(
      Array Sampling Mode=::NewEnumerator0
      Direct Array Index=?
      Color Selection Array=DataInterface() {propertyValues {ColorData [<rgba 0.107 0.3737 1.5 1>, <rgba 1 1 1 0.5>, <rgba 0.273 1 0.6852 0.75>]}}
      Random Seed=0
     )
     "Option 4 (Blue)":LinearColor=&SelectLinearColorFromArray(
      Array Sampling Mode=::NewEnumerator0
      Direct Array Index=?
      Color Selection Array=DataInterface() {propertyValues {ColorData [<rgba 0.0552 1 0.7486 0.5>, <rgba 1 1 1 1>, <rgba 1.072 0.5461 2 1>]}}
      Random Seed=0
     )
     Color Select:NiagaraInt32=$Particles.BurstColor
    )
   ) {isSetParameters=true}
   /SetVariables(Particles.Color=&MaskLinearColorBySpawnGroup(Masked Linear Color=$Transient.ThinColor, Passthrough Linear Color=$Transient.ThickColor)) {isSetParameters=true}
  ]
  particleUpdateScript [
   &ParticleState()
   &ScaleMeshSize(Scale Factor=&VectorFromFloat(Value=&FloatFromCurve(
      FloatCurve=DataInterface() {propertyValues {Curve {keys [
          {arriveTangent=2.3265, interpMode=RCIM_Cubic, leaveTangent=2.3265, tangentMode=RCTM_User, time=0, value=0}
          {tangentMode=RCTM_User, time=1, value=1}
         ]}}}
      Scale Curve=&MaskFloatBySpawnGroup(
       Masked Float=&RandomRangeFloat(Minimum=0.2, Maximum=0.4)
       Passthrough Float=&RandomRangeFloat(Minimum=0.25, Maximum=0.3)
      )
     )))
   /SetVariables(
    Transient.ThickThickness:NiagaraFloat=&FloatFromCurve(FloatCurve=DataInterface() {propertyValues {Curve {
        defaultValue=1
        keys [
         {arriveTangent=1.0303, interpMode=RCIM_Cubic, leaveTangent=1.0303, leaveTangentWeight=0.0571, tangentMode=RCTM_User, time=0, value=0.5}
         {arriveTangentWeight=0.2577, interpMode=RCIM_Cubic, leaveTangentWeight=0.0859, tangentMode=RCTM_User, time=1, value=1}
        ]
       }}})
    Transient.ThinThickness:NiagaraFloat=&Lerp_Float(
     A=&RandomRangeFloat(Minimum=0.3, Maximum=0.6)
     B=0.05
     Alpha=&FloatFromCurve(FloatCurve=DataInterface() {propertyValues {Curve {
         defaultValue=1
         keys [
          {arriveTangent=1.6743, interpMode=RCIM_Cubic, leaveTangent=1.6743, leaveTangentWeight=0.0571, tangentMode=RCTM_User, time=0, value=0}
          {arriveTangent=1.1664, interpMode=RCIM_Cubic, leaveTangent=1.1664, tangentMode=RCTM_User, time=0.5, value=0.7}
          {arriveTangent=0.4285, tangentMode=RCTM_User, time=1, value=1}
         ]
        }}})
    )
   ) {isSetParameters=true}
   &DynamicMaterialParameters(
    Index 0 Param 1=&FloatFromCurve(
     FloatCurve=DataInterface() {propertyValues {Curve {keys [
         {arriveTangent=11.7429, interpMode=RCIM_Cubic, leaveTangent=11.7429, tangentMode=RCTM_User, time=0, value=1}
         {arriveTangent=-0.7624, interpMode=RCIM_Cubic, leaveTangent=-0.7624, tangentMode=RCTM_User, time=0.2, value=2}
         {interpMode=RCIM_Cubic, tangentMode=RCTM_User, time=1, value=1}
        ]}}}
     Scale Curve=&MakeCustomFloatFromBool(Bool=$User.IsFrontend, True Float=1, False Float=2)
    )
    Index 0 Param 2=&MaskFloatBySpawnGroup(Masked Float=$Transient.ThinThickness, Passthrough Float=$Transient.ThickThickness)
    Index 0 Param 3=&FloatFromCurve(FloatCurve=DataInterface() {propertyValues {Curve {keys [
         {arriveTangent=0.585, interpMode=RCIM_Cubic, leaveTangent=0.585, tangentMode=RCTM_User, time=0, value=0}
         {arriveTangent=2.0717, tangentMode=RCTM_User, time=1, value=1}
        ]}}})
    Index 1 Vector 1:Vector3f=&RotateVectorByQuaternion(Vector To Rotate=<0 1 0>, Quaternion=$Engine.Owner.Rotation)
    Module.Index0Float0Write=true
    Module.Index0Float1Write=true
    Module.Index0Float2Write=true
    Module.Index0Float3Write=false
    Module.Index1Float3Write=false
    Module.Index1VectorWrite=true
    Write Parameter Index 1=::NewEnumerator3
   )
   /SetVariables(
    Transient.ThickColor:LinearColor=&self:SelectColor(
     "Option 0 (Red)":LinearColor=&ColorFromCurve(ColorCurve=DataInterface() {propertyValues {
        AlphaCurve {keys [{interpMode=RCIM_Cubic, tangentMode=RCTM_User, time=0, value=1}]}
        BlueCurve {keys [{interpMode=RCIM_Cubic, leaveTangentWeight=0.0854, tangentMode=RCTM_User, time=0, value=0.2174}, <at 1 0.2227>]}
        GreenCurve {keys [{interpMode=RCIM_Cubic, leaveTangentWeight=0.0854, tangentMode=RCTM_User, time=0, value=0.2174}, <at 1 0.0517>]}
        RedCurve {keys [{interpMode=RCIM_Cubic, leaveTangentWeight=0.0854, tangentMode=RCTM_User, time=0, value=1}, <at 1 0.3861>]}
       }})
     "Option 1 (Orange)":LinearColor=&ColorFromCurve(ColorCurve=DataInterface() {propertyValues {
        AlphaCurve {keys [{interpMode=RCIM_Cubic, tangentMode=RCTM_User, time=0, value=1}]}
        BlueCurve {keys [
          {interpMode=RCIM_Cubic, leaveTangentWeight=0.0854, tangentMode=RCTM_User, time=0, value=0.1374}
          {
           arriveTangent=-0.179
           arriveTangentWeight=0.0901
           interpMode=RCIM_Cubic
           leaveTangent=-0.179
           leaveTangentWeight=0.136
           tangentMode=RCTM_User
           time=0.2562
           value=0.0643
          }
          {arriveTangentWeight=0.2195, interpMode=RCIM_Cubic, leaveTangentWeight=0.2026, tangentMode=RCTM_User, time=0.643, value=0.0296}
          {arriveTangentWeight=0.119, interpMode=RCIM_Cubic, tangentMode=RCTM_User, time=1, value=0.0296}
         ]}
        GreenCurve {keys [
          {interpMode=RCIM_Cubic, leaveTangentWeight=0.0854, tangentMode=RCTM_User, time=0, value=1}
          {
           arriveTangent=-0.6584
           arriveTangentWeight=0.0854
           interpMode=RCIM_Cubic
           leaveTangent=-0.6584
           leaveTangentWeight=0.1289
           tangentMode=RCTM_User
           time=0.2562
           value=0.8649
          }
          {
           arriveTangent=-0.9079
           arriveTangentWeight=0.1289
           interpMode=RCIM_Cubic
           leaveTangent=-0.9079
           leaveTangentWeight=0.119
           tangentMode=RCTM_User
           time=0.643
           value=0.5764
          }
          {arriveTangent=-0.9999, arriveTangentWeight=0.119, interpMode=RCIM_Cubic, leaveTangent=-0.9999, tangentMode=RCTM_User, time=1, value=0.1713}
         ]}
        RedCurve {keys [
          {interpMode=RCIM_Cubic, leaveTangentWeight=0.0854, tangentMode=RCTM_User, time=0, value=0.9985}
          {arriveTangentWeight=0.1404, interpMode=RCIM_Cubic, leaveTangentWeight=0.2119, tangentMode=RCTM_User, time=0.2562, value=1}
          {arriveTangentWeight=0.1303, interpMode=RCIM_Cubic, leaveTangentWeight=0.1202, tangentMode=RCTM_User, time=0.643, value=1}
          {arriveTangentWeight=0.119, interpMode=RCIM_Cubic, tangentMode=RCTM_User, time=1, value=1}
         ]}
       }})
     "Option 2 (Yellow)":LinearColor=&ColorFromCurve(ColorCurve=DataInterface() {propertyValues {
        AlphaCurve {keys [{interpMode=RCIM_Cubic, tangentMode=RCTM_User, time=0, value=1}]}
        BlueCurve {keys [
          {interpMode=RCIM_Cubic, leaveTangentWeight=0.0854, tangentMode=RCTM_User, time=0, value=0.1374}
          {
           arriveTangent=-0.179
           arriveTangentWeight=0.0901
           interpMode=RCIM_Cubic
           leaveTangent=-0.179
           leaveTangentWeight=0.136
           tangentMode=RCTM_User
           time=0.2562
           value=0.0643
          }
          {arriveTangentWeight=0.2195, interpMode=RCIM_Cubic, leaveTangentWeight=0.2026, tangentMode=RCTM_User, time=0.643, value=0.0296}
          {arriveTangentWeight=0.119, interpMode=RCIM_Cubic, tangentMode=RCTM_User, time=1, value=0.0296}
         ]}
        GreenCurve {keys [
          {interpMode=RCIM_Cubic, leaveTangentWeight=0.0854, tangentMode=RCTM_User, time=0, value=1}
          {
           arriveTangent=-0.6584
           arriveTangentWeight=0.0854
           interpMode=RCIM_Cubic
           leaveTangent=-0.6584
           leaveTangentWeight=0.1289
           tangentMode=RCTM_User
           time=0.2562
           value=0.8649
          }
          {
           arriveTangent=-0.9079
           arriveTangentWeight=0.1289
           interpMode=RCIM_Cubic
           leaveTangent=-0.9079
           leaveTangentWeight=0.119
           tangentMode=RCTM_User
           time=0.643
           value=0.5764
          }
          {arriveTangent=-0.9999, arriveTangentWeight=0.119, interpMode=RCIM_Cubic, leaveTangent=-0.9999, tangentMode=RCTM_User, time=1, value=0.1713}
         ]}
        RedCurve {keys [
          {interpMode=RCIM_Cubic, leaveTangentWeight=0.0854, tangentMode=RCTM_User, time=0, value=0.9985}
          {arriveTangentWeight=0.1404, interpMode=RCIM_Cubic, leaveTangentWeight=0.2119, tangentMode=RCTM_User, time=0.2562, value=1}
          {arriveTangentWeight=0.1303, interpMode=RCIM_Cubic, leaveTangentWeight=0.1202, tangentMode=RCTM_User, time=0.643, value=1}
          {arriveTangentWeight=0.119, interpMode=RCIM_Cubic, tangentMode=RCTM_User, time=1, value=1}
         ]}
       }})
     "Option 3 (Green)":LinearColor=&ColorFromCurve(ColorCurve=DataInterface() {propertyValues {
        AlphaCurve {keys [{interpMode=RCIM_Cubic, tangentMode=RCTM_User, time=0, value=1}]}
        BlueCurve {keys [
          {interpMode=RCIM_Cubic, leaveTangentWeight=0.0854, tangentMode=RCTM_User, time=0, value=0.1026}
          {
           arriveTangent=0.3367
           arriveTangentWeight=0.0929
           interpMode=RCIM_Cubic
           leaveTangent=0.3367
           leaveTangentWeight=0.1403
           tangentMode=RCTM_User
           time=0.2562
           value=0.0643
          }
          {arriveTangent=0.6385, interpMode=RCIM_Cubic, leaveTangent=0.6385, time=0.6442, value=0.1574}
          {arriveTangentWeight=0.119, interpMode=RCIM_Cubic, tangentMode=RCTM_User, time=1, value=0.5392}
         ]}
        GreenCurve {keys [
          {interpMode=RCIM_Cubic, leaveTangentWeight=0.0854, tangentMode=RCTM_User, time=0, value=1}
          {arriveTangentWeight=0.1337, interpMode=RCIM_Cubic, leaveTangentWeight=0.2018, tangentMode=RCTM_User, time=0.2562, value=1}
          <at 0.6442 1>
          {arriveTangentWeight=0.119, interpMode=RCIM_Cubic, tangentMode=RCTM_User, time=1, value=1}
         ]}
        RedCurve {keys [
          {interpMode=RCIM_Cubic, leaveTangentWeight=0.0854, tangentMode=RCTM_User, time=0, value=0.9984}
          {
           arriveTangent=-1.3049
           arriveTangentWeight=0.0893
           interpMode=RCIM_Cubic
           leaveTangent=-1.3049
           leaveTangentWeight=0.1348
           tangentMode=RCTM_User
           time=0.2562
           value=0.2368
          }
          <at 0.6442 0.0093>
          {arriveTangentWeight=0.119, interpMode=RCIM_Cubic, tangentMode=RCTM_User, time=1, value=0.0296}
         ]}
       }})
     "Option 4 (Blue)":LinearColor=&ColorFromCurve(ColorCurve=DataInterface() {propertyValues {
        AlphaCurve {keys [<at 0 1 RCIM_Cubic>]}
        BlueCurve {keys [
          {interpMode=RCIM_Cubic, leaveTangentWeight=0.0854, time=0, value=0.9794}
          {
           arriveTangent=0.0206
           arriveTangentWeight=0.0929
           interpMode=RCIM_Cubic
           leaveTangent=0.0206
           leaveTangentWeight=0.1403
           time=0.4982
           value=1
          }
          {arriveTangentWeight=0.1471, interpMode=RCIM_Cubic, leaveTangent=-0.0051, leaveTangentWeight=0.1358, time=1, value=1}
         ]}
        GreenCurve {keys [
          {interpMode=RCIM_Cubic, leaveTangentWeight=0.0854, time=0, value=1.25}
          {
           arriveTangent=-0.8642
           arriveTangentWeight=0.1337
           interpMode=RCIM_Cubic
           leaveTangent=-0.8642
           leaveTangentWeight=0.2018
           time=0.4982
           value=0.9854
          }
          {arriveTangentWeight=0.1514, interpMode=RCIM_Cubic, leaveTangent=-0.8098, leaveTangentWeight=0.1398, time=1, value=0.3858}
         ]}
        RedCurve {keys [
          {interpMode=RCIM_Cubic, leaveTangentWeight=0.0854, time=0, value=0.9065}
          {
           arriveTangent=-0.557
           arriveTangentWeight=0.0893
           interpMode=RCIM_Cubic
           leaveTangent=-0.557
           leaveTangentWeight=0.1348
           time=0.4982
           value=0.44
          }
          {arriveTangentWeight=0.233, interpMode=RCIM_Cubic, leaveTangent=0.859, leaveTangentWeight=0.2151, time=1, value=0.3496}
         ]}
       }})
     Color Select:NiagaraInt32=$Particles.BurstColor
    )
    Transient.ThinColor:LinearColor=&self:SelectColor(
     "Option 0 (Red)":LinearColor=&ColorFromCurve(ColorCurve=DataInterface() {propertyValues {
        AlphaCurve {keys [{interpMode=RCIM_Cubic, tangentMode=RCTM_User, time=0, value=1}]}
        BlueCurve {keys [{interpMode=RCIM_Cubic, leaveTangentWeight=0.0854, tangentMode=RCTM_User, time=0, value=0.3009}, <at 1 0.1724>]}
        GreenCurve {keys [{interpMode=RCIM_Cubic, leaveTangentWeight=0.0854, tangentMode=RCTM_User, time=0, value=0.3009}, <at 1 0.1017>]}
        RedCurve {keys [{interpMode=RCIM_Cubic, leaveTangentWeight=0.0854, tangentMode=RCTM_User, time=0, value=1}, <at 1 0.5>]}
       }})
     "Option 1 (Orange)":LinearColor=&ColorFromCurve(ColorCurve=DataInterface() {propertyValues {
        AlphaCurve {keys [{interpMode=RCIM_Cubic, tangentMode=RCTM_User, time=0, value=1}]}
        BlueCurve {keys [
          {arriveTangent=-0.6993, interpMode=RCIM_Cubic, leaveTangent=-0.6993, leaveTangentWeight=0.0854, tangentMode=RCTM_User, time=0, value=0.4969}
          {
           arriveTangent=-0.0562
           arriveTangentWeight=0.2195
           interpMode=RCIM_Cubic
           leaveTangent=-0.0562
           leaveTangentWeight=0.2026
           tangentMode=RCTM_User
           time=0.643
           value=0.0574
          }
          {arriveTangentWeight=0.119, interpMode=RCIM_Cubic, tangentMode=RCTM_User, time=1, value=0.0296}
         ]}
        GreenCurve {keys [
          {arriveTangent=-0.242, interpMode=RCIM_Cubic, leaveTangent=-0.242, leaveTangentWeight=0.0854, tangentMode=RCTM_User, time=0, value=0.8644}
          {
           arriveTangent=-0.6793
           arriveTangentWeight=0.1289
           interpMode=RCIM_Cubic
           leaveTangent=-0.6793
           leaveTangentWeight=0.119
           tangentMode=RCTM_User
           time=0.643
           value=0.5885
          }
          {arriveTangent=-0.9999, arriveTangentWeight=0.119, interpMode=RCIM_Cubic, leaveTangent=-0.9999, tangentMode=RCTM_User, time=1, value=0.1713}
         ]}
        RedCurve {keys [
          {interpMode=RCIM_Cubic, leaveTangentWeight=0.0854, tangentMode=RCTM_User, time=0, value=1}
          {arriveTangentWeight=0.1303, interpMode=RCIM_Cubic, leaveTangentWeight=0.1202, tangentMode=RCTM_User, time=0.643, value=1}
          {arriveTangentWeight=0.119, interpMode=RCIM_Cubic, tangentMode=RCTM_User, time=1, value=1}
         ]}
       }})
     "Option 2 (Yellow)":LinearColor=&ColorFromCurve(ColorCurve=DataInterface() {propertyValues {
        AlphaCurve {keys [{interpMode=RCIM_Cubic, tangentMode=RCTM_User, time=0, value=1}]}
        BlueCurve {keys [
          {interpMode=RCIM_Cubic, leaveTangentWeight=0.0854, tangentMode=RCTM_User, time=0, value=0.3878}
          {
           arriveTangent=-0.7021
           arriveTangentWeight=0.0901
           interpMode=RCIM_Cubic
           leaveTangent=-0.7021
           leaveTangentWeight=0.136
           tangentMode=RCTM_User
           time=0.2562
           value=0.0643
          }
          {arriveTangentWeight=0.2195, interpMode=RCIM_Cubic, leaveTangentWeight=0.2026, tangentMode=RCTM_User, time=0.643, value=0.0574}
          {arriveTangentWeight=0.119, interpMode=RCIM_Cubic, tangentMode=RCTM_User, time=1, value=0.0296}
         ]}
        GreenCurve {keys [
          {interpMode=RCIM_Cubic, leaveTangentWeight=0.0854, tangentMode=RCTM_User, time=0, value=1}
          {
           arriveTangent=-0.6584
           arriveTangentWeight=0.0854
           interpMode=RCIM_Cubic
           leaveTangent=-0.6584
           leaveTangentWeight=0.1289
           tangentMode=RCTM_User
           time=0.2562
           value=0.8649
          }
          {
           arriveTangent=-0.9079
           arriveTangentWeight=0.1289
           interpMode=RCIM_Cubic
           leaveTangent=-0.9079
           leaveTangentWeight=0.119
           tangentMode=RCTM_User
           time=0.643
           value=0.5885
          }
          {arriveTangent=-0.9999, arriveTangentWeight=0.119, interpMode=RCIM_Cubic, leaveTangent=-0.9999, tangentMode=RCTM_User, time=1, value=0.1713}
         ]}
        RedCurve {keys [
          {interpMode=RCIM_Cubic, leaveTangentWeight=0.0854, tangentMode=RCTM_User, time=0, value=0.9989}
          {arriveTangentWeight=0.1404, interpMode=RCIM_Cubic, leaveTangentWeight=0.2119, tangentMode=RCTM_User, time=0.2562, value=1}
          {arriveTangentWeight=0.1303, interpMode=RCIM_Cubic, leaveTangentWeight=0.1202, tangentMode=RCTM_User, time=0.643, value=1}
          {arriveTangentWeight=0.119, interpMode=RCIM_Cubic, tangentMode=RCTM_User, time=1, value=1}
         ]}
       }})
     "Option 3 (Green)":LinearColor=&ColorFromCurve(ColorCurve=DataInterface() {propertyValues {
        AlphaCurve {keys [{interpMode=RCIM_Cubic, tangentMode=RCTM_User, time=0, value=1}]}
        BlueCurve {keys [
          {interpMode=RCIM_Cubic, leaveTangentWeight=0.0854, tangentMode=RCTM_User, time=0, value=0.1235}
          {
           arriveTangent=0.3367
           arriveTangentWeight=0.0929
           interpMode=RCIM_Cubic
           leaveTangent=0.3367
           leaveTangentWeight=0.1403
           tangentMode=RCTM_User
           time=0.2562
           value=0.0643
          }
          {
           arriveTangent=1.3779
           arriveTangentWeight=0.1471
           interpMode=RCIM_Cubic
           leaveTangent=1.3779
           leaveTangentWeight=0.1358
           tangentMode=RCTM_User
           time=0.643
           value=0.0584
          }
          {arriveTangentWeight=0.119, interpMode=RCIM_Cubic, tangentMode=RCTM_User, time=1, value=0.2962}
         ]}
        GreenCurve {keys [
          {interpMode=RCIM_Cubic, leaveTangentWeight=0.0854, tangentMode=RCTM_User, time=0, value=1}
          {arriveTangentWeight=0.1337, interpMode=RCIM_Cubic, leaveTangentWeight=0.2018, tangentMode=RCTM_User, time=0.2562, value=1}
          {arriveTangentWeight=0.1514, interpMode=RCIM_Cubic, leaveTangentWeight=0.1398, tangentMode=RCTM_User, time=0.643, value=0.727}
          {arriveTangentWeight=0.119, interpMode=RCIM_Cubic, tangentMode=RCTM_User, time=1, value=1}
         ]}
        RedCurve {keys [
          {interpMode=RCIM_Cubic, leaveTangentWeight=0.0854, tangentMode=RCTM_User, time=0, value=0.9985}
          {
           arriveTangent=-1.3049
           arriveTangentWeight=0.0893
           interpMode=RCIM_Cubic
           leaveTangent=-1.3049
           leaveTangentWeight=0.1348
           tangentMode=RCTM_User
           time=0.2562
           value=0.2368
          }
          {
           arriveTangent=-0.1445
           arriveTangentWeight=0.233
           interpMode=RCIM_Cubic
           leaveTangent=-0.1445
           leaveTangentWeight=0.2151
           tangentMode=RCTM_User
           time=0.643
           value=0.0417
          }
          {arriveTangentWeight=0.119, interpMode=RCIM_Cubic, tangentMode=RCTM_User, time=1, value=0.0296}
         ]}
       }})
     "Option 4 (Blue)":LinearColor=&ColorFromCurve(ColorCurve=DataInterface() {propertyValues {
        AlphaCurve {keys [{interpMode=RCIM_Cubic, tangentMode=RCTM_User, time=0, value=1}]}
        BlueCurve {keys [
          {interpMode=RCIM_Cubic, leaveTangentWeight=0.0854, tangentMode=RCTM_User, time=0, value=0.3095}
          {
           arriveTangent=0.3367
           arriveTangentWeight=0.0929
           interpMode=RCIM_Cubic
           leaveTangent=0.3367
           leaveTangentWeight=0.1403
           tangentMode=RCTM_User
           time=0.2562
           value=1
          }
          {
           arriveTangent=1.3779
           arriveTangentWeight=0.1471
           interpMode=RCIM_Cubic
           leaveTangent=1.3779
           leaveTangentWeight=0.1358
           tangentMode=RCTM_User
           time=0.643
           value=0.727
          }
          {arriveTangentWeight=0.119, interpMode=RCIM_Cubic, tangentMode=RCTM_User, time=1, value=1}
         ]}
        GreenCurve {keys [
          {interpMode=RCIM_Cubic, leaveTangentWeight=0.0854, tangentMode=RCTM_User, time=0, value=1}
          {arriveTangentWeight=0.1337, interpMode=RCIM_Cubic, leaveTangentWeight=0.2018, tangentMode=RCTM_User, time=0.2562, value=0.9756}
          {arriveTangentWeight=0.1514, interpMode=RCIM_Cubic, leaveTangentWeight=0.1398, tangentMode=RCTM_User, time=0.643, value=0.0799}
          {arriveTangentWeight=0.119, interpMode=RCIM_Cubic, tangentMode=RCTM_User, time=1, value=0.0296}
         ]}
        RedCurve {keys [
          {interpMode=RCIM_Cubic, leaveTangentWeight=0.0854, tangentMode=RCTM_User, time=0, value=0.1235}
          {
           arriveTangent=-1.3049
           arriveTangentWeight=0.0893
           interpMode=RCIM_Cubic
           leaveTangent=-1.3049
           leaveTangentWeight=0.1348
           tangentMode=RCTM_User
           time=0.2562
           value=0.0643
          }
          {
           arriveTangent=-0.1445
           arriveTangentWeight=0.233
           interpMode=RCIM_Cubic
           leaveTangent=-0.1445
           leaveTangentWeight=0.2151
           tangentMode=RCTM_User
           time=0.643
           value=0.0417
          }
          {arriveTangentWeight=0.119, interpMode=RCIM_Cubic, tangentMode=RCTM_User, time=1, value=0.5021}
         ]}
       }})
     Color Select:NiagaraInt32=$Particles.BurstColor
    )
   ) {isSetParameters=true}
   &ScaleColor(
    Scale RGB=&Multiply_VectorByFloat(
     Vector=&MakeVectorFromLinearColorRGB(Color=&MaskLinearColorBySpawnGroup(Masked Linear Color=$Transient.ThinColor, Passthrough Linear Color=$Transient.ThickColor))
     Float=&MaskFloatBySpawnGroup(
      Masked Float=&FloatFromCurve(
       FloatCurve=DataInterface() {propertyValues {Curve {keys [<at 0 1 RCIM_Cubic>, {arriveTangent=6.1862, tangentMode=RCTM_User, time=1, value=3}]}}}
       Scale Curve=&MakeCustomFloatFromBool(Bool=$User.IsFrontend, True Float=2, False Float=3)
      )
      Passthrough Float=&FloatFromCurve(
       FloatCurve=DataInterface() {propertyValues {Curve {keys [<at 0 1 RCIM_Cubic>, {arriveTangent=-1.8064, tangentMode=RCTM_User, time=1, value=0}]}}}
       Scale Curve=&MakeCustomFloatFromBool(Bool=$User.IsFrontend, True Float=1, False Float=2)
      )
     )
    )
    ScaleA=false
    ScaleRGB=true
   )
   &CameraOffset(Camera Offset Amount=-50)
  ]
  rendererClasses [&NiagaraSpriteRendererProperties, &NiagaraMeshRendererProperties]
  renderers [
   {rendererClass=&NiagaraSpriteRendererProperties}
   {rendererClass=&NiagaraMeshRendererProperties, rendererIndex=1}
   {rendererClass=&NiagaraMeshRendererProperties, rendererIndex=2}
  ]
 }
}
summary {
 systemName=NS_Aura_Squares
 emitters [
  {emitterName=Decal, rendererClasses [&NiagaraSpriteRendererProperties]}
  {emitterName=Squares, rendererClasses [&NiagaraSpriteRendererProperties, &NiagaraMeshRendererProperties]}
  {emitterName=Glows, rendererClasses [&NiagaraSpriteRendererProperties]}
 ]
 userVariables [
  User.IsFrontend {type=NiagaraBool}
  User.IsLargeBody {type=NiagaraBool}
 ]
}
