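# Niagara system for the start burst of the Squares aura cosmetic, placed on a skeletal mesh.
# One-shot Corona, Dark_Glow and Flare sprites give a short blue flash, with Flare also driving a separate flash color and size.
# Squares bursts two spawn groups of growing thin and thick shapes, colored from five random palettes and shaped through dynamic material parameters.
# User.IsFrontEnd shrinks and fades the squares; the sparks emitter is disabled.
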NiagaraSystem /SparksCosmetics/Items/Auras/Assets/NS_Aura_Squares/NS_Aura_Squares_Start
dependencies {
 usedDataInterfaces [
  &NiagaraDataInterfaceCurve
  &NiagaraDataInterfaceArrayColor
  &NiagaraDataInterfaceColorCurve
  &NiagaraDataInterfaceSkeletalMesh
 ]
 usedDynamicInputs [&MakeFloatFromVector]
 usedModules [
  &self:SystemSpawnScript.SystemScriptSource.SystemScriptGraph.NiagaraNodeAssignment_0.SetVariables_#1
  &SystemState
 ]
}
emitters {
 Corona {
  emitterUpdateScript [
   &EmitterState()
   &SpawnBurst_Instantaneous(Spawn Count=1, Use Spawn Probability=false)
  ]
  particleSpawnScript [
   &InitializeParticle(
    Color=<rgba 0 0.4115 1 1>
    Lifetime=0.25
    Sprite Rotation Mode=::NewEnumerator2
    Sprite Size Mode=::NewEnumerator3
    Uniform Sprite Size=400
    UsePositionOffset=false
    Write Material Random=true
   )
   &DynamicMaterialParameters(
    Index 0 Param 1=&RandomRangeFloat(Minimum=0.4, Maximum=0.6)
    Index 0 Param 2=1
    Index 0 Param 3=1
    Index 0 Param 4=1
    Module.Param0WriteEnabled=true
    Module.Param1WriteEnabled=true
    Module.Param2WriteEnabled=true
    Module.Param3WriteEnabled=true
   )
   &SystemLocation(MaskSpawnGroup=false, Offset=<0 0 120>)
  ]
  particleUpdateScript [
   &ParticleState()
   &SolveForcesAndVelocity(Clamp Velocity=false, Limit Acceleration=false)
   &SkeletalMeshLocation(Bone Sampling Mode=::NewEnumerator4, MaskSpawnGroup=false, Skeletal Mesh=$System.Skeletal Mesh) {module=SkeletalMeshLocation001}
   &ScaleColor(
    Scale Alpha=&Multiply_Float(
     A=&FNTimeOfDayScalar(Night Scale=1, Morning Scale=1, Evening Scale=1)
     B=&FloatFromCurve(FloatCurve=DataInterface() {propertyValues {Curve {keys [
          {arriveTangent=9.6062, interpMode=RCIM_Cubic, leaveTangent=9.6062, tangentMode=RCTM_User, time=0, value=0}
          {interpMode=RCIM_Cubic, tangentMode=RCTM_User, time=0.2, value=1}
          {interpMode=RCIM_Cubic, tangentMode=RCTM_User, time=1, value=0}
         ]}}})
    )
    Scale RGB=&Multiply_VectorByFloat(Vector=<1 1 1>, Float=1)
    ScaleA=true
    ScaleRGB=true
   )
   &ScaleSpriteSize(Scale Factor=&Vector2DFromFloat(Value=&FloatFromCurve(FloatCurve=DataInterface() {propertyValues {Curve {keys [
          {arriveTangent=2.1722, interpMode=RCIM_Cubic, leaveTangent=2.1722, tangentMode=RCTM_User, time=0, value=0}
          {arriveTangent=0.0224, interpMode=RCIM_Cubic, leaveTangent=0.0224, tangentMode=RCTM_User, time=1, value=1}
         ]}}})))
   &CameraOffset(Camera Offset Amount=-30)
   &DynamicMaterialParameters(
    Index 0 Param 1=0.75
    Index 0 Param 2=2
    Index 0 Param 3=1
    Module.Param0WriteEnabled=true
    Module.Param1WriteEnabled=true
    Module.Param2WriteEnabled=true
    Module.Param3WriteEnabled=true
   ) {module=DynamicMaterialParameters001}
  ]
  rendererClasses [&NiagaraSpriteRendererProperties]
  renderers [{rendererClass=&NiagaraSpriteRendererProperties}]
 }
 Dark_Glow {
  emitterUpdateScript [
   &EmitterState()
   &SpawnBurst_Instantaneous(Spawn Count=1, Use Spawn Probability=false)
  ]
  particleSpawnScript [&InitializeParticle(
    Color=<rgba 0 0.0823 0.2 1>
    Lifetime=0.4
    Position Offset=<0 0 100>
    Sprite Rotation Mode=::NewEnumerator2
    Sprite Size Mode=::NewEnumerator3
    Uniform Sprite Size=230
    UsePositionOffset=true
    Write Material Random=true
   )]
  particleUpdateScript [
   &ParticleState()
   &SkeletalMeshLocation(Bone Sampling Mode=::NewEnumerator4, MaskSpawnGroup=false, Skeletal Mesh=$System.Skeletal Mesh) {module=SkeletalMeshLocation001}
   &ScaleColor(
    Scale Alpha=&FloatFromCurve(
     FloatCurve=DataInterface() {propertyValues {Curve {keys [
         {interpMode=RCIM_Cubic, tangentMode=RCTM_User, time=0, value=0}
         {interpMode=RCIM_Cubic, tangentMode=RCTM_User, time=0.2, value=1}
         {interpMode=RCIM_Cubic, tangentMode=RCTM_User, time=0.5, value=1}
         {tangentMode=RCTM_User, time=1, value=0}
        ]}}}
     Scale Curve=2
    )
    ScaleA=true
    ScaleRGB=true
   )
  ]
  rendererClasses [&NiagaraSpriteRendererProperties]
  renderers [{rendererClass=&NiagaraSpriteRendererProperties}]
 }
 Flare {
  emitterUpdateScript [
   &EmitterState()
   &SpawnBurst_Instantaneous(Spawn Count=1, Use Spawn Probability=false)
  ]
  particleSpawnScript [
   &InitializeParticle(
    Color=<rgba 0 0.5143 1.25 1>
    Lifetime=0.3
    Sprite Rotation Mode=::NewEnumerator2
    Sprite Size Mode=::NewEnumerator3
    Uniform Sprite Size=600
    UsePositionOffset=false
    Write Material Random=true
   )
   &DynamicMaterialParameters(
    Index 0 Param 1=&RandomRangeFloat(Minimum=0.4, Maximum=0.6)
    Index 0 Param 2=1
    Index 0 Param 3=1
    Index 0 Param 4=1
    Module.Param0WriteEnabled=true
    Module.Param1WriteEnabled=true
    Module.Param2WriteEnabled=true
    Module.Param3WriteEnabled=true
   )
   &SystemLocation(MaskSpawnGroup=false, Offset=<0 0 120>)
   /SetVariables(Particles.Sharp Flare Camera Offset=40) {isSetParameters=true}
   &CameraOffset(Camera Offset Amount=-30) {module=CameraOffset001}
  ]
  particleUpdateScript [
   &ParticleState()
   &SolveForcesAndVelocity(Clamp Velocity=false, Limit Acceleration=false)
   &SkeletalMeshLocation(Bone Sampling Mode=::NewEnumerator4, MaskSpawnGroup=false, Skeletal Mesh=$System.Skeletal Mesh) {module=SkeletalMeshLocation001}
   &ScaleColor(
    Scale Alpha=&Multiply_Float(
     A=&FNTimeOfDayScalar(Night Scale=1, Morning Scale=1, Evening Scale=1)
     B=&FloatFromCurve(FloatCurve=DataInterface() {propertyValues {Curve {keys [<at 0 0>,<at 0.2 1>,<at 0.8 1>,<at 1 0>]}}})
    )
    Scale RGB=&Multiply_VectorByFloat(Vector=<1 1 1>, Float=1)
    ScaleA=true
    ScaleRGB=true
   )
   &ScaleSpriteSize(Scale Factor=&Vector2DFromFloat(Value=&FloatFromCurve(FloatCurve=DataInterface() {propertyValues {Curve {keys [<at 0 1.5 RCIM_Cubic>,<at 1 0>]}}})))
   /SetVariables(Particles.Flash color:LinearColor=&MakeLinearColorFromVectorAndFloat(
     "Vector (RGB)"=&MakeVectorFromLinearColorRGB(Color=<rgba 0.5235 0.3277 1 1>)
     "Float (Alpha)"=&Multiply_Float(
      A=&Multiply_Float(
       A=&FNTimeOfDayScalar(Night Scale=1, Morning Scale=1, Evening Scale=1)
       B=&FloatFromCurve(FloatCurve=DataInterface() {propertyValues {Curve {keys [<at 0 0>,<at 0.124 1>,<at 0.496 1>,<at 0.62 0>]}}})
      )
      B=1
     )
    )) {isSetParameters=true}
   /SetVariables(Particles.Flash size:Vector2f=&Multiply_Vector2D_ByFloat(
     Float=&Multiply_Float(
      A=400
      B=&FloatFromCurve(FloatCurve=DataInterface() {propertyValues {Curve {keys [
           {arriveTangent=4.1785, interpMode=RCIM_Cubic, leaveTangent=4.1785, tangentMode=RCTM_User, time=0, value=0}
           {arriveTangent=-0.15, interpMode=RCIM_Cubic, leaveTangent=-0.15, tangentMode=RCTM_User, time=0.1378, value=1}
           {arriveTangent=-0.0405, interpMode=RCIM_Cubic, leaveTangent=-0.0405, tangentMode=RCTM_User, time=0.62, value=0.25}
          ]}}})
     )
     Vector2D=<1 1>
    )) {isSetParameters=true}
  ]
  rendererClasses [&NiagaraSpriteRendererProperties]
  renderers [
   {rendererClass=&NiagaraSpriteRendererProperties}
   {rendererClass=&NiagaraSpriteRendererProperties, rendererIndex=1}
  ]
 }
 Sparks_Burst001 {
  bEnabled=false
  emitterUpdateScript [
   &EmitterState(Life Cycle Mode=::NewEnumerator1, Loop Behavior=::NewEnumerator1, Loop Duration=0.15, UseLoopDelay=false)
   &SpawnRate(SpawnRate=250, Use Spawn Probability=false)
  ]
  particleSpawnScript [
   &InitializeParticle(
    Color=&SelectLinearColorFromArray(
     Array Sampling Mode=::NewEnumerator0
     Direct Array Index=?
     Color Selection Array=DataInterface() {propertyValues {ColorData [<rgba 0 0.3702 1 1>, <rgba 0 1 0.8813 1>]}}
     Random Seed=0
    )
    Lifetime Max=1.75
    Lifetime Mode=::NewEnumerator1
    Position Mode=::NewEnumerator1
    Sprite Rotation Mode=::NewEnumerator2
    Sprite Size=&Lerp_Vector2(
     A=<1 3>
     B=<2 5>
     Alpha=&FloatFromCurve(
      FloatCurve=DataInterface() {propertyValues {Curve {keys [
          {interpMode=RCIM_Cubic, tangentMode=RCTM_User, time=0, value=0}
          {interpMode=RCIM_Cubic, tangentMode=RCTM_User, time=0.3, value=0}
          {interpMode=RCIM_Cubic, tangentMode=RCTM_User, time=0.8, value=1}
         ]}}}
      CurveIndex=&RandomRangeFloat(Minimum=0, Maximum=1)
     )
    )
    Sprite Size Mode=::NewEnumerator4
    UsePositionOffset=false
    Write Material Random=true
   ) {module=InitializeParticle001}
   &ShapeLocation(
    Apply To Particle Position=::NewEnumerator1
    MaskSpawnGroup=false
    Override Seed=false
    Shape Origin=&VectorFromFloat(Value=0)
    Sphere Radius=10
    UseOwnerScale=true
   )
   &AddVelocity(
    Velocity=&MakeVector(
     X=&RandomRangeFloat(Minimum=-400, Maximum=400)
     Y=&RandomRangeFloat(Minimum=-400, Maximum=400)
     Z=&RandomRangeFloat(Minimum=-400, Maximum=400)
    )
    Velocity Speed Scale=3
   ) {module=AddVelocity001}
  ]
  particleUpdateScript [
   &ParticleState()
   &ScaleColor(
    Scale Alpha=&Lerp_Float(
     A=1
     B=0
     Alpha=&FloatFromCurve(FloatCurve=DataInterface() {propertyValues {Curve {keys [
          {arriveTangent=3.1144, interpMode=RCIM_Cubic, leaveTangent=3.1144, tangentMode=RCTM_User, time=0, value=0}
          {interpMode=RCIM_Cubic, tangentMode=RCTM_User, time=1, value=1}
         ]}}})
    )
    Scale RGB=&Multiply_VectorByFloat(
     Vector=<1 1 1>
     Float=&FloatFromCurve(
      FloatCurve=DataInterface() {propertyValues {Curve {keys [
          {arriveTangent=-12.6044, interpMode=RCIM_Cubic, leaveTangent=-12.6044, tangentMode=RCTM_User, time=0, value=5}
          {interpMode=RCIM_Cubic, tangentMode=RCTM_User, time=1, value=1}
         ]}}}
      Scale Curve=10
     )
    )
    ScaleA=true
    ScaleRGB=true
   )
   &ScaleSpriteSize(Uniform Curve Sprite Scale=DataInterface() {propertyValues {Curve {keys [
        {arriveTangent=0.0565, interpMode=RCIM_Cubic, leaveTangent=0.0565, tangentMode=RCTM_User, time=0, value=1}
        {arriveTangent=-0.025, interpMode=RCIM_Cubic, leaveTangent=-0.025, tangentMode=RCTM_User, time=1, value=0}
       ]}}}) {module=ScaleSpriteSize001}
   &SkeletalMeshLocation(
    Bone / Socket Index=$Particles.SpawnGroup
    Bone Sampling Mode=::NewEnumerator4
    MaskSpawnGroup=false
    Position Sampling=::NewEnumerator1
    Sampled Position Offset=$StackContext.ShapeLocation.ShapePosition
    Skeletal Mesh=$System.Skeletal Mesh
    Source Movement Scale:Vector3f=<0 0 0>
   ) {module=SkeletalMeshLocation001}
   &Drag(Drag=10, Use Linear Drag=true, Use Rotational Drag=false)
   &CurlNoiseForce(Noise Frequency=15, Noise Strength=500, Pan Noise=true, Pan Noise Field=<0.25 0.25 0.25>)
   &SolveForcesAndVelocity(Clamp Velocity=false, Limit Acceleration=false)
  ]
  rendererClasses [&NiagaraSpriteRendererProperties]
  renderers [{rendererClass=&NiagaraSpriteRendererProperties}]
 }
 Squares {
  emitterUpdateScript [
   &EmitterState()
   &SpawnBurst_Instantaneous(Spawn Count=3, Spawn Group=1, Use Spawn Probability=false, UseLoopCountLimit=false)
   &SpawnBurst_Instantaneous(Spawn Count=5, Use Spawn Probability=false, UseLoopCountLimit=false) {module=SpawnBurst_Instantaneous001}
  ]
  particleSpawnScript [
   &InitializeParticle(
    Lifetime=&MaskFloatBySpawnGroup(
     Masked Float=&RandomRangeFloat(Minimum=0.375, Maximum=1)
     Passthrough Float=&RandomRangeFloat(Minimum=0.375, Maximum=0.625)
    )
    Mesh Scale Mode=::NewEnumerator3
    Mesh Uniform Scale=5
    UsePositionOffset=false
    Write Material Random=true
   )
   /SetVariables(Particles.BurstColor:NiagaraInt32=&RandomRangeInt(Minimum=0, Maximum=4)) {isSetParameters=true}
   /SetVariables(
    Transient.ThickColor:LinearColor=&self:SelectColor(
     "Option 0 (Red)":LinearColor=&SelectLinearColorFromArray(
      Array Sampling Mode=::NewEnumerator0
      Direct Array Index=?
      Color Selection Array=DataInterface() {propertyValues {ColorData [<rgba 1.7 0.1705 0.1705 1>, <rgba 1 0.1478 0.1478 1>]}}
      Random Seed=0
     )
     "Option 1 (Orange)":LinearColor=&SelectLinearColorFromArray(
      Array Sampling Mode=::NewEnumerator0
      Direct Array Index=?
      Color Selection Array=DataInterface() {propertyValues {ColorData [<rgba 1 0.6664 0.0552 0.6>, <rgba 1 0.247 0.0226 0.75>, <rgba 1 0.3555 0.2452 0.6>]}}
      Random Seed=0
     )
     "Option 2 (Yellow)":LinearColor=&SelectLinearColorFromArray(
      Array Sampling Mode=::NewEnumerator0
      Direct Array Index=?
      Color Selection Array=DataInterface() {propertyValues {ColorData [<rgba 1 0.6664 0.0552 0.4>, <rgba 1 1 1 0.25>, <rgba 0.4789 1 0.2452 0.3>]}}
      Random Seed=0
     )
     "Option 3 (Green)":LinearColor=&SelectLinearColorFromArray(
      Array Sampling Mode=::NewEnumerator0
      Direct Array Index=?
      Color Selection Array=DataInterface() {propertyValues {ColorData [<rgba 0.1357 0.3395 1.2 1>, <rgba 1 1 1 0.3>, <rgba 0.273 1 0.6852 0.3>]}}
      Random Seed=0
     )
     "Option 4 (Blue)":LinearColor=&SelectLinearColorFromArray(
      Array Sampling Mode=::NewEnumerator0
      Direct Array Index=?
      Color Selection Array=DataInterface() {propertyValues {ColorData [<rgba 0.3922 0.2452 1 0.75>, <rgba 1 1 1 0.3>, <rgba 0.1687 0.8742 1 0.5>]}}
      Random Seed=0
     )
     Color Select:NiagaraInt32=$Particles.BurstColor
    )
    Transient.ThinColor:LinearColor=&self:SelectColor(
     "Option 0 (Red)":LinearColor=&SelectLinearColorFromArray(
      Array Sampling Mode=::NewEnumerator0
      Direct Array Index=?
      Color Selection Array=DataInterface() {propertyValues {ColorData [<rgba 2 0.2006 0.2306 1>, <rgba 1.5 0.3854 0.347 1>]}}
      Random Seed=0
     )
     "Option 1 (Orange)":LinearColor=&SelectLinearColorFromArray(
      Array Sampling Mode=::NewEnumerator0
      Direct Array Index=?
      Color Selection Array=DataInterface() {propertyValues {ColorData [<rgba 1.5 0.349 0.0828 1>, <rgba 1 0.4464 0.051 0.2>, <rgba 1 0.2875 0.1617 1>]}}
      Random Seed=0
     )
     "Option 2 (Yellow)":LinearColor=&SelectLinearColorFromArray(
      Array Sampling Mode=::NewEnumerator0
      Direct Array Index=?
      Color Selection Array=DataInterface() {propertyValues {ColorData [<rgba 1 0.6664 0.0552 1>, <rgba 1 0.7434 0.273 0.5>, <rgba 0.6195 1 0.273 0.5>]}}
      Random Seed=0
     )
     "Option 3 (Green)":LinearColor=&SelectLinearColorFromArray(
      Array Sampling Mode=::NewEnumerator0
      Direct Array Index=?
      Color Selection Array=DataInterface() {propertyValues {ColorData [<rgba 0.107 0.3737 1.5 1>, <rgba 1 1 1 0.5>, <rgba 0.273 1 0.6852 0.75>]}}
      Random Seed=0
     )
     "Option 4 (Blue)":LinearColor=&SelectLinearColorFromArray(
      Array Sampling Mode=::NewEnumerator0
      Direct Array Index=?
      Color Selection Array=DataInterface() {propertyValues {ColorData [<rgba 0.0552 1 0.7486 0.5>, <rgba 1 1 1 1>, <rgba 1.072 0.5461 2 1>]}}
      Random Seed=0
     )
     Color Select:NiagaraInt32=$Particles.BurstColor
    )
   ) {isSetParameters=true}
   /SetVariables(Particles.Color=&MaskLinearColorBySpawnGroup(Masked Linear Color=$Transient.ThinColor, Passthrough Linear Color=$Transient.ThickColor)) {isSetParameters=true}
   &SkeletalMeshLocation(Bone Sampling Mode=::NewEnumerator4, MaskSpawnGroup=false, Orientation Sampling=::NewEnumerator1, Skeletal Mesh=$System.Skeletal Mesh) {module=SkeletalMeshLocation001}
  ]
  particleUpdateScript [
   &ParticleState()
   &ScaleMeshSize(Scale Factor=&VectorFromFloat(Value=&Multiply_Float(
      A=&FloatFromCurve(
       FloatCurve=DataInterface() {propertyValues {Curve {keys [
           {arriveTangent=2.3265, interpMode=RCIM_Cubic, leaveTangent=2.3265, tangentMode=RCTM_User, time=0, value=0}
           {tangentMode=RCTM_User, time=1, value=1}
          ]}}}
       Scale Curve=&MaskFloatBySpawnGroup(
        Masked Float=&RandomRangeFloat(Minimum=0.2, Maximum=0.75)
        Passthrough Float=&RandomRangeFloat(Minimum=0.25, Maximum=0.3)
       )
      )
      B=&MakeCustomFloatFromBool(Bool=$User.IsFrontEnd, True Float=0.8, False Float=1)
     )))
   /SetVariables(
    Transient.ThickThickness:NiagaraFloat=&FloatFromCurve(FloatCurve=DataInterface() {propertyValues {Curve {
        defaultValue=1
        keys [
         {arriveTangent=1.0303, interpMode=RCIM_Cubic, leaveTangent=1.0303, leaveTangentWeight=0.0571, tangentMode=RCTM_User, time=0, value=0.5}
         {arriveTangentWeight=0.2577, interpMode=RCIM_Cubic, leaveTangentWeight=0.0859, tangentMode=RCTM_User, time=1, value=1}
        ]
       }}})
    Transient.ThinThickness:NiagaraFloat=&Lerp_Float(
     A=&RandomRangeFloat(Minimum=0.3, Maximum=0.6)
     B=0.01
     Alpha=&FloatFromCurve(FloatCurve=DataInterface() {propertyValues {Curve {
         defaultValue=1
         keys [
          {arriveTangent=1.6743, interpMode=RCIM_Cubic, leaveTangent=1.6743, leaveTangentWeight=0.0571, tangentMode=RCTM_User, time=0, value=0}
          {arriveTangent=1.1664, interpMode=RCIM_Cubic, leaveTangent=1.1664, tangentMode=RCTM_User, time=0.5, value=0.7}
          {arriveTangent=0.4285, tangentMode=RCTM_User, time=1, value=1}
         ]
        }}})
    )
   ) {isSetParameters=true}
   &DynamicMaterialParameters(
    Index 0 Param 1=&FloatFromCurve(FloatCurve=DataInterface() {propertyValues {Curve {keys [
         {arriveTangent=11.7429, interpMode=RCIM_Cubic, leaveTangent=11.7429, tangentMode=RCTM_User, time=0, value=1}
         {arriveTangent=-0.7624, interpMode=RCIM_Cubic, leaveTangent=-0.7624, tangentMode=RCTM_User, time=0.2, value=2}
         {interpMode=RCIM_Cubic, tangentMode=RCTM_User, time=1, value=1}
        ]}}})
    Index 0 Param 2=&MaskFloatBySpawnGroup(Masked Float=$Transient.ThinThickness, Passthrough Float=$Transient.ThickThickness)
    Index 0 Param 3=&FloatFromCurve(FloatCurve=DataInterface() {propertyValues {Curve {keys [
         {arriveTangent=0.585, interpMode=RCIM_Cubic, leaveTangent=0.585, tangentMode=RCTM_User, time=0, value=0}
         {arriveTangent=2.0717, tangentMode=RCTM_User, time=1, value=1}
        ]}}})
    Index 1 Vector 1:Vector3f=&RotateVectorByQuaternion(Vector To Rotate=<0 1 0>, Quaternion=$Engine.Owner.Rotation)
    Module.Index0Float0Write=true
    Module.Index0Float1Write=true
    Module.Index0Float2Write=true
    Module.Index0Float3Write=false
    Module.Index1Float3Write=false
    Module.Index1VectorWrite=true
    Write Parameter Index 1=::NewEnumerator3
   )
   /SetVariables(
    Transient.ThickColor:LinearColor=&self:SelectColor(
     "Option 0 (Red)":LinearColor=&ColorFromCurve(ColorCurve=DataInterface() {propertyValues {
        AlphaCurve {keys [{interpMode=RCIM_Cubic, tangentMode=RCTM_User, time=0, value=1}]}
        BlueCurve {keys [{interpMode=RCIM_Cubic, leaveTangentWeight=0.0854, tangentMode=RCTM_User, time=0, value=0.2174}, <at 1 0.2227>]}
        GreenCurve {keys [{interpMode=RCIM_Cubic, leaveTangentWeight=0.0854, tangentMode=RCTM_User, time=0, value=0.2174}, <at 1 0.0517>]}
        RedCurve {keys [{interpMode=RCIM_Cubic, leaveTangentWeight=0.0854, tangentMode=RCTM_User, time=0, value=1}, <at 1 0.3861>]}
       }})
     "Option 1 (Orange)":LinearColor=&ColorFromCurve(ColorCurve=DataInterface() {propertyValues {
        AlphaCurve {keys [{interpMode=RCIM_Cubic, tangentMode=RCTM_User, time=0, value=1}]}
        BlueCurve {keys [
          {interpMode=RCIM_Cubic, leaveTangentWeight=0.0854, tangentMode=RCTM_User, time=0, value=0.1374}
          {
           arriveTangent=-0.179
           arriveTangentWeight=0.0901
           interpMode=RCIM_Cubic
           leaveTangent=-0.179
           leaveTangentWeight=0.136
           tangentMode=RCTM_User
           time=0.2562
           value=0.0643
          }
          {arriveTangentWeight=0.2195, interpMode=RCIM_Cubic, leaveTangentWeight=0.2026, tangentMode=RCTM_User, time=0.643, value=0.0296}
          {arriveTangentWeight=0.119, interpMode=RCIM_Cubic, tangentMode=RCTM_User, time=1, value=0.0296}
         ]}
        GreenCurve {keys [
          {interpMode=RCIM_Cubic, leaveTangentWeight=0.0854, tangentMode=RCTM_User, time=0, value=1}
          {
           arriveTangent=-0.6584
           arriveTangentWeight=0.0854
           interpMode=RCIM_Cubic
           leaveTangent=-0.6584
           leaveTangentWeight=0.1289
           tangentMode=RCTM_User
           time=0.2562
           value=0.8649
          }
          {
           arriveTangent=-0.9079
           arriveTangentWeight=0.1289
           interpMode=RCIM_Cubic
           leaveTangent=-0.9079
           leaveTangentWeight=0.119
           tangentMode=RCTM_User
           time=0.643
           value=0.5764
          }
          {arriveTangent=-0.9999, arriveTangentWeight=0.119, interpMode=RCIM_Cubic, leaveTangent=-0.9999, tangentMode=RCTM_User, time=1, value=0.1713}
         ]}
        RedCurve {keys [
          {interpMode=RCIM_Cubic, leaveTangentWeight=0.0854, tangentMode=RCTM_User, time=0, value=0.9985}
          {arriveTangentWeight=0.1404, interpMode=RCIM_Cubic, leaveTangentWeight=0.2119, tangentMode=RCTM_User, time=0.2562, value=1}
          {arriveTangentWeight=0.1303, interpMode=RCIM_Cubic, leaveTangentWeight=0.1202, tangentMode=RCTM_User, time=0.643, value=1}
          {arriveTangentWeight=0.119, interpMode=RCIM_Cubic, tangentMode=RCTM_User, time=1, value=1}
         ]}
       }})
     "Option 2 (Yellow)":LinearColor=&ColorFromCurve(ColorCurve=DataInterface() {propertyValues {
        AlphaCurve {keys [{interpMode=RCIM_Cubic, tangentMode=RCTM_User, time=0, value=1}]}
        BlueCurve {keys [
          {interpMode=RCIM_Cubic, leaveTangentWeight=0.0854, tangentMode=RCTM_User, time=0, value=0.1374}
          {
           arriveTangent=-0.179
           arriveTangentWeight=0.0901
           interpMode=RCIM_Cubic
           leaveTangent=-0.179
           leaveTangentWeight=0.136
           tangentMode=RCTM_User
           time=0.2562
           value=0.0643
          }
          {arriveTangentWeight=0.2195, interpMode=RCIM_Cubic, leaveTangentWeight=0.2026, tangentMode=RCTM_User, time=0.643, value=0.0296}
          {arriveTangentWeight=0.119, interpMode=RCIM_Cubic, tangentMode=RCTM_User, time=1, value=0.0296}
         ]}
        GreenCurve {keys [
          {interpMode=RCIM_Cubic, leaveTangentWeight=0.0854, tangentMode=RCTM_User, time=0, value=1}
          {
           arriveTangent=-0.6584
           arriveTangentWeight=0.0854
           interpMode=RCIM_Cubic
           leaveTangent=-0.6584
           leaveTangentWeight=0.1289
           tangentMode=RCTM_User
           time=0.2562
           value=0.8649
          }
          {
           arriveTangent=-0.9079
           arriveTangentWeight=0.1289
           interpMode=RCIM_Cubic
           leaveTangent=-0.9079
           leaveTangentWeight=0.119
           tangentMode=RCTM_User
           time=0.643
           value=0.5764
          }
          {arriveTangent=-0.9999, arriveTangentWeight=0.119, interpMode=RCIM_Cubic, leaveTangent=-0.9999, tangentMode=RCTM_User, time=1, value=0.1713}
         ]}
        RedCurve {keys [
          {interpMode=RCIM_Cubic, leaveTangentWeight=0.0854, tangentMode=RCTM_User, time=0, value=0.9985}
          {arriveTangentWeight=0.1404, interpMode=RCIM_Cubic, leaveTangentWeight=0.2119, tangentMode=RCTM_User, time=0.2562, value=1}
          {arriveTangentWeight=0.1303, interpMode=RCIM_Cubic, leaveTangentWeight=0.1202, tangentMode=RCTM_User, time=0.643, value=1}
          {arriveTangentWeight=0.119, interpMode=RCIM_Cubic, tangentMode=RCTM_User, time=1, value=1}
         ]}
       }})
     "Option 3 (Green)":LinearColor=&ColorFromCurve(ColorCurve=DataInterface() {propertyValues {
        AlphaCurve {keys [{interpMode=RCIM_Cubic, tangentMode=RCTM_User, time=0, value=1}]}
        BlueCurve {keys [
          {interpMode=RCIM_Cubic, leaveTangentWeight=0.0854, tangentMode=RCTM_User, time=0, value=0.1026}
          {
           arriveTangent=0.3367
           arriveTangentWeight=0.0929
           interpMode=RCIM_Cubic
           leaveTangent=0.3367
           leaveTangentWeight=0.1403
           tangentMode=RCTM_User
           time=0.2562
           value=0.0643
          }
          {arriveTangent=0.6385, interpMode=RCIM_Cubic, leaveTangent=0.6385, time=0.6442, value=0.1574}
          {arriveTangentWeight=0.119, interpMode=RCIM_Cubic, tangentMode=RCTM_User, time=1, value=0.5392}
         ]}
        GreenCurve {keys [
          {interpMode=RCIM_Cubic, leaveTangentWeight=0.0854, tangentMode=RCTM_User, time=0, value=1}
          {arriveTangentWeight=0.1337, interpMode=RCIM_Cubic, leaveTangentWeight=0.2018, tangentMode=RCTM_User, time=0.2562, value=1}
          <at 0.6442 1>
          {arriveTangentWeight=0.119, interpMode=RCIM_Cubic, tangentMode=RCTM_User, time=1, value=1}
         ]}
        RedCurve {keys [
          {interpMode=RCIM_Cubic, leaveTangentWeight=0.0854, tangentMode=RCTM_User, time=0, value=0.9984}
          {
           arriveTangent=-1.3049
           arriveTangentWeight=0.0893
           interpMode=RCIM_Cubic
           leaveTangent=-1.3049
           leaveTangentWeight=0.1348
           tangentMode=RCTM_User
           time=0.2562
           value=0.2368
          }
          <at 0.6442 0.0093>
          {arriveTangentWeight=0.119, interpMode=RCIM_Cubic, tangentMode=RCTM_User, time=1, value=0.0296}
         ]}
       }})
     "Option 4 (Blue)":LinearColor=&ColorFromCurve(ColorCurve=DataInterface() {propertyValues {
        AlphaCurve {keys [<at 0 1 RCIM_Cubic>]}
        BlueCurve {keys [
          {interpMode=RCIM_Cubic, leaveTangentWeight=0.0854, time=0, value=0.9794}
          {
           arriveTangent=0.0206
           arriveTangentWeight=0.0929
           interpMode=RCIM_Cubic
           leaveTangent=0.0206
           leaveTangentWeight=0.1403
           time=0.4982
           value=1
          }
          {arriveTangentWeight=0.1471, interpMode=RCIM_Cubic, leaveTangent=-0.0051, leaveTangentWeight=0.1358, time=1, value=1}
         ]}
        GreenCurve {keys [
          {interpMode=RCIM_Cubic, leaveTangentWeight=0.0854, time=0, value=1.25}
          {
           arriveTangent=-0.8642
           arriveTangentWeight=0.1337
           interpMode=RCIM_Cubic
           leaveTangent=-0.8642
           leaveTangentWeight=0.2018
           time=0.4982
           value=0.9854
          }
          {arriveTangentWeight=0.1514, interpMode=RCIM_Cubic, leaveTangent=-0.8098, leaveTangentWeight=0.1398, time=1, value=0.3858}
         ]}
        RedCurve {keys [
          {interpMode=RCIM_Cubic, leaveTangentWeight=0.0854, time=0, value=0.9065}
          {
           arriveTangent=-0.557
           arriveTangentWeight=0.0893
           interpMode=RCIM_Cubic
           leaveTangent=-0.557
           leaveTangentWeight=0.1348
           time=0.4982
           value=0.44
          }
          {arriveTangentWeight=0.233, interpMode=RCIM_Cubic, leaveTangent=0.859, leaveTangentWeight=0.2151, time=1, value=0.3496}
         ]}
       }})
     Color Select:NiagaraInt32=$Particles.BurstColor
    )
    Transient.ThinColor:LinearColor=&self:SelectColor(
     "Option 0 (Red)":LinearColor=&ColorFromCurve(ColorCurve=DataInterface() {propertyValues {
        AlphaCurve {keys [{interpMode=RCIM_Cubic, tangentMode=RCTM_User, time=0, value=1}]}
        BlueCurve {keys [{interpMode=RCIM_Cubic, leaveTangentWeight=0.0854, tangentMode=RCTM_User, time=0, value=0.3009}, <at 1 0.1724>]}
        GreenCurve {keys [{interpMode=RCIM_Cubic, leaveTangentWeight=0.0854, tangentMode=RCTM_User, time=0, value=0.3009}, <at 1 0.1017>]}
        RedCurve {keys [{interpMode=RCIM_Cubic, leaveTangentWeight=0.0854, tangentMode=RCTM_User, time=0, value=1}, <at 1 0.5>]}
       }})
     "Option 1 (Orange)":LinearColor=&ColorFromCurve(ColorCurve=DataInterface() {propertyValues {
        AlphaCurve {keys [{interpMode=RCIM_Cubic, tangentMode=RCTM_User, time=0, value=1}]}
        BlueCurve {keys [
          {arriveTangent=-0.6993, interpMode=RCIM_Cubic, leaveTangent=-0.6993, leaveTangentWeight=0.0854, tangentMode=RCTM_User, time=0, value=0.4969}
          {
           arriveTangent=-0.0562
           arriveTangentWeight=0.2195
           interpMode=RCIM_Cubic
           leaveTangent=-0.0562
           leaveTangentWeight=0.2026
           tangentMode=RCTM_User
           time=0.643
           value=0.0574
          }
          {arriveTangentWeight=0.119, interpMode=RCIM_Cubic, tangentMode=RCTM_User, time=1, value=0.0296}
         ]}
        GreenCurve {keys [
          {arriveTangent=-0.242, interpMode=RCIM_Cubic, leaveTangent=-0.242, leaveTangentWeight=0.0854, tangentMode=RCTM_User, time=0, value=0.8644}
          {
           arriveTangent=-0.6793
           arriveTangentWeight=0.1289
           interpMode=RCIM_Cubic
           leaveTangent=-0.6793
           leaveTangentWeight=0.119
           tangentMode=RCTM_User
           time=0.643
           value=0.5885
          }
          {arriveTangent=-0.9999, arriveTangentWeight=0.119, interpMode=RCIM_Cubic, leaveTangent=-0.9999, tangentMode=RCTM_User, time=1, value=0.1713}
         ]}
        RedCurve {keys [
          {interpMode=RCIM_Cubic, leaveTangentWeight=0.0854, tangentMode=RCTM_User, time=0, value=1}
          {arriveTangentWeight=0.1303, interpMode=RCIM_Cubic, leaveTangentWeight=0.1202, tangentMode=RCTM_User, time=0.643, value=1}
          {arriveTangentWeight=0.119, interpMode=RCIM_Cubic, tangentMode=RCTM_User, time=1, value=1}
         ]}
       }})
     "Option 2 (Yellow)":LinearColor=&ColorFromCurve(ColorCurve=DataInterface() {propertyValues {
        AlphaCurve {keys [{interpMode=RCIM_Cubic, tangentMode=RCTM_User, time=0, value=1}]}
        BlueCurve {keys [
          {interpMode=RCIM_Cubic, leaveTangentWeight=0.0854, tangentMode=RCTM_User, time=0, value=0.3878}
          {
           arriveTangent=-0.7021
           arriveTangentWeight=0.0901
           interpMode=RCIM_Cubic
           leaveTangent=-0.7021
           leaveTangentWeight=0.136
           tangentMode=RCTM_User
           time=0.2562
           value=0.0643
          }
          {arriveTangentWeight=0.2195, interpMode=RCIM_Cubic, leaveTangentWeight=0.2026, tangentMode=RCTM_User, time=0.643, value=0.0574}
          {arriveTangentWeight=0.119, interpMode=RCIM_Cubic, tangentMode=RCTM_User, time=1, value=0.0296}
         ]}
        GreenCurve {keys [
          {interpMode=RCIM_Cubic, leaveTangentWeight=0.0854, tangentMode=RCTM_User, time=0, value=1}
          {
           arriveTangent=-0.6584
           arriveTangentWeight=0.0854
           interpMode=RCIM_Cubic
           leaveTangent=-0.6584
           leaveTangentWeight=0.1289
           tangentMode=RCTM_User
           time=0.2562
           value=0.8649
          }
          {
           arriveTangent=-0.9079
           arriveTangentWeight=0.1289
           interpMode=RCIM_Cubic
           leaveTangent=-0.9079
           leaveTangentWeight=0.119
           tangentMode=RCTM_User
           time=0.643
           value=0.5885
          }
          {arriveTangent=-0.9999, arriveTangentWeight=0.119, interpMode=RCIM_Cubic, leaveTangent=-0.9999, tangentMode=RCTM_User, time=1, value=0.1713}
         ]}
        RedCurve {keys [
          {interpMode=RCIM_Cubic, leaveTangentWeight=0.0854, tangentMode=RCTM_User, time=0, value=0.9989}
          {arriveTangentWeight=0.1404, interpMode=RCIM_Cubic, leaveTangentWeight=0.2119, tangentMode=RCTM_User, time=0.2562, value=1}
          {arriveTangentWeight=0.1303, interpMode=RCIM_Cubic, leaveTangentWeight=0.1202, tangentMode=RCTM_User, time=0.643, value=1}
          {arriveTangentWeight=0.119, interpMode=RCIM_Cubic, tangentMode=RCTM_User, time=1, value=1}
         ]}
       }})
     "Option 3 (Green)":LinearColor=&ColorFromCurve(ColorCurve=DataInterface() {propertyValues {
        AlphaCurve {keys [{interpMode=RCIM_Cubic, tangentMode=RCTM_User, time=0, value=1}]}
        BlueCurve {keys [
          {interpMode=RCIM_Cubic, leaveTangentWeight=0.0854, tangentMode=RCTM_User, time=0, value=0.1235}
          {
           arriveTangent=0.3367
           arriveTangentWeight=0.0929
           interpMode=RCIM_Cubic
           leaveTangent=0.3367
           leaveTangentWeight=0.1403
           tangentMode=RCTM_User
           time=0.2562
           value=0.0643
          }
          {
           arriveTangent=1.3779
           arriveTangentWeight=0.1471
           interpMode=RCIM_Cubic
           leaveTangent=1.3779
           leaveTangentWeight=0.1358
           tangentMode=RCTM_User
           time=0.643
           value=0.0584
          }
          {arriveTangentWeight=0.119, interpMode=RCIM_Cubic, tangentMode=RCTM_User, time=1, value=0.2962}
         ]}
        GreenCurve {keys [
          {interpMode=RCIM_Cubic, leaveTangentWeight=0.0854, tangentMode=RCTM_User, time=0, value=1}
          {arriveTangentWeight=0.1337, interpMode=RCIM_Cubic, leaveTangentWeight=0.2018, tangentMode=RCTM_User, time=0.2562, value=1}
          {arriveTangentWeight=0.1514, interpMode=RCIM_Cubic, leaveTangentWeight=0.1398, tangentMode=RCTM_User, time=0.643, value=0.727}
          {arriveTangentWeight=0.119, interpMode=RCIM_Cubic, tangentMode=RCTM_User, time=1, value=1}
         ]}
        RedCurve {keys [
          {interpMode=RCIM_Cubic, leaveTangentWeight=0.0854, tangentMode=RCTM_User, time=0, value=0.9985}
          {
           arriveTangent=-1.3049
           arriveTangentWeight=0.0893
           interpMode=RCIM_Cubic
           leaveTangent=-1.3049
           leaveTangentWeight=0.1348
           tangentMode=RCTM_User
           time=0.2562
           value=0.2368
          }
          {
           arriveTangent=-0.1445
           arriveTangentWeight=0.233
           interpMode=RCIM_Cubic
           leaveTangent=-0.1445
           leaveTangentWeight=0.2151
           tangentMode=RCTM_User
           time=0.643
           value=0.0417
          }
          {arriveTangentWeight=0.119, interpMode=RCIM_Cubic, tangentMode=RCTM_User, time=1, value=0.0296}
         ]}
       }})
     "Option 4 (Blue)":LinearColor=&ColorFromCurve(ColorCurve=DataInterface() {propertyValues {
        AlphaCurve {keys [{interpMode=RCIM_Cubic, tangentMode=RCTM_User, time=0, value=1}]}
        BlueCurve {keys [
          {interpMode=RCIM_Cubic, leaveTangentWeight=0.0854, tangentMode=RCTM_User, time=0, value=0.3095}
          {
           arriveTangent=0.3367
           arriveTangentWeight=0.0929
           interpMode=RCIM_Cubic
           leaveTangent=0.3367
           leaveTangentWeight=0.1403
           tangentMode=RCTM_User
           time=0.2562
           value=1
          }
          {
           arriveTangent=1.3779
           arriveTangentWeight=0.1471
           interpMode=RCIM_Cubic
           leaveTangent=1.3779
           leaveTangentWeight=0.1358
           tangentMode=RCTM_User
           time=0.643
           value=0.727
          }
          {arriveTangentWeight=0.119, interpMode=RCIM_Cubic, tangentMode=RCTM_User, time=1, value=1}
         ]}
        GreenCurve {keys [
          {interpMode=RCIM_Cubic, leaveTangentWeight=0.0854, tangentMode=RCTM_User, time=0, value=1}
          {arriveTangentWeight=0.1337, interpMode=RCIM_Cubic, leaveTangentWeight=0.2018, tangentMode=RCTM_User, time=0.2562, value=0.9756}
          {arriveTangentWeight=0.1514, interpMode=RCIM_Cubic, leaveTangentWeight=0.1398, tangentMode=RCTM_User, time=0.643, value=0.0799}
          {arriveTangentWeight=0.119, interpMode=RCIM_Cubic, tangentMode=RCTM_User, time=1, value=0.0296}
         ]}
        RedCurve {keys [
          {interpMode=RCIM_Cubic, leaveTangentWeight=0.0854, tangentMode=RCTM_User, time=0, value=0.1235}
          {
           arriveTangent=-1.3049
           arriveTangentWeight=0.0893
           interpMode=RCIM_Cubic
           leaveTangent=-1.3049
           leaveTangentWeight=0.1348
           tangentMode=RCTM_User
           time=0.2562
           value=0.0643
          }
          {
           arriveTangent=-0.1445
           arriveTangentWeight=0.233
           interpMode=RCIM_Cubic
           leaveTangent=-0.1445
           leaveTangentWeight=0.2151
           tangentMode=RCTM_User
           time=0.643
           value=0.0417
          }
          {arriveTangentWeight=0.119, interpMode=RCIM_Cubic, tangentMode=RCTM_User, time=1, value=0.5021}
         ]}
       }})
     Color Select:NiagaraInt32=$Particles.BurstColor
    )
   ) {isSetParameters=true}
   &ScaleColor(
    Scale Alpha=&MakeCustomFloatFromBool(Bool=$User.IsFrontEnd, True Float=0.25, False Float=1)
    Scale RGB=&Multiply_VectorByFloat(
     Vector=&MakeVectorFromLinearColorRGB(Color=&MaskLinearColorBySpawnGroup(Masked Linear Color=$Transient.ThinColor, Passthrough Linear Color=$Transient.ThickColor))
     Float=&MaskFloatBySpawnGroup(
      Masked Float=&FloatFromCurve(
       FloatCurve=DataInterface() {propertyValues {Curve {keys [<at 0 1 RCIM_Cubic>, {arriveTangent=6.1862, tangentMode=RCTM_User, time=1, value=3}]}}}
       Scale Curve=2
      )
      Passthrough Float=&FloatFromCurve(FloatCurve=DataInterface() {propertyValues {Curve {keys [<at 0 1 RCIM_Cubic>, {arriveTangent=-1.8064, tangentMode=RCTM_User, time=1, value=0}]}}})
     )
    )
    ScaleA=true
    ScaleRGB=true
   )
   &CameraOffset(Camera Offset Amount=-50)
  ]
  rendererClasses [&NiagaraSpriteRendererProperties, &NiagaraMeshRendererProperties]
  renderers [
   {rendererClass=&NiagaraSpriteRendererProperties}
   {rendererClass=&NiagaraMeshRendererProperties, rendererIndex=1}
   {rendererClass=&NiagaraMeshRendererProperties, rendererIndex=2}
  ]
 }
}
summary {
 systemName=NS_Aura_Squares_Start
 emitters [
  {emitterName=Flare, rendererClasses [&NiagaraSpriteRendererProperties]}
  {emitterName=Dark_Glow, rendererClasses [&NiagaraSpriteRendererProperties]}
  {emitterName=Corona, rendererClasses [&NiagaraSpriteRendererProperties]}
  {bEnabled=false, emitterName=Sparks_Burst001, rendererClasses [&NiagaraSpriteRendererProperties]}
  {emitterName=Squares, rendererClasses [&NiagaraSpriteRendererProperties, &NiagaraMeshRendererProperties]}
 ]
 userVariables [User.IsFrontEnd {type=NiagaraBool}]
}
